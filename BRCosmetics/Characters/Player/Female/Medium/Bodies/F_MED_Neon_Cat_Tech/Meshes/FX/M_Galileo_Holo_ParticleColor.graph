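# Unlit translucent Galileo hologram material for skeletal meshes and Niagara particles, with optional particle color and alpha tinting.
# Fresnel, distance-scaled scanlines, flickering world-aligned noise, vertical gradient and TurnOn pulse build a mask that a curve atlas colors.
# Static switches select 2D logo, VendettaPuck profile/slate image, or camera-facing bobbing variants.
# Noise also drives vertex displacement; opacity adds depth fade, custom-depth occlusion and optional strobe.
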Material /BRCosmetics/Characters/Player/Female/Medium/Bodies/F_MED_Neon_Cat_Tech/Meshes/FX/M_Galileo_Holo_ParticleColor
graph {
 properties {
  allowTranslucentCustomDepthWrites=true
  bEnableResponsiveAA=true
  bUsedWithMeshDeformer=true
  bUsedWithNiagaraMeshParticles=true
  bUsedWithNiagaraSprites=true
  bUsedWithSkeletalMesh=true
  blendMode=BLEND_Translucent
  pixelDepthOffsetMode=PDOM_Legacy
  shadingModel=MSM_Unlit
  usedShadingModels=Unlit
 }
 nodes [
  n1 = VectorParameter() {parameterName=LightDirection, defaultValue=<rgba 0 0 1 1>}
  n2 = TextureSampleParameter2D() {group=Texture, parameterName=Normals, samplerType=SAMPLERTYPE_Normal, texture=/Engine/EngineMaterials/BaseFlattenNormalMap}
  n3 = LinearInterpolate(A=VertexNormalWS(), B=Transform(None=@n2.RGB), Alpha=ScalarParameter() {group=Lighting, parameterName=UseNormal})
  n4 = Multiply(A=Clamp(None=DotProduct(A=@n1.RGB, B=@n3)), B=ScalarParameter() {defaultValue=0.2, group=Lighting, parameterName=NormalBrightnessAdd}) {constB=0.15}
  n5 = TextureSampleParameter2D() {group=Texture, parameterName=M, texture=/Engine/EngineResources/WhiteSquareTexture}
  n6 = ScalarParameter() {defaultValue=1, parameterName=TurnOn}
  n7 = Divide(A=ObjectRadius(), B=ScalarParameter() {defaultValue=0.1, group=Noise, parameterName=NoiseSize})
  n8 = Multiply(A=Time(), B=VectorParameter() {group=Noise, parameterName=NoisePanSpeed, defaultValue=<rgba -0.08 0.07 -0.2 0>}.RGB)
  n9 = MaterialFunctionCall(
   TextureObject=TextureObject() {samplerType=SAMPLERTYPE_LinearColor, texture=&T_NoisePack_LinearColor}
   TextureSize=Multiply(A=VectorParameter() {group=Noise, parameterName=NoiseTileMult, defaultValue=<rgba 1 1 0.1 1>}.RGB, B=@n7)
   WorldPosition=Add(A=WorldPosition().XYZ, B=Multiply(A=@n7, B=@n8))
  ) {materialFunction=&WorldAlignedTexture}
  n10 = ActorPositionWS()
  n11 = MaterialFunctionCall(Float2=Round(None=Divide(A=@n10) {constB=2000})) {materialFunction=&BreakOutFloat2Components}
  n12 = Add(A=Time(), B=Add(A=@n11.R, B=@n11.R))
  n13 = Multiply(A=ScalarParameter() {defaultValue=0.5, parameterName=FlickerSpeed}, B=@n12)
  n14 = AppendVector(A=Sine(None=@n13) {period=1.6}, B=Sine(None=@n13) {period=1.42})
  n15 = Sine(None=@n13) {period=1.51}
  n16 = Sine(None=@n13) {period=2.1}
  n17 = ScalarParameter() {defaultValue=0.5, group=Noise, parameterName=NoiseFrequency}
  n18 = Multiply(B=@n17) {constA=0.25}
  n19 = Add(A=Sine(None=@n13) {period=0.72}, B=Multiply(B=@n17) {constA=0.75}) {constB=0.75}
  n20 = Add(A=Multiply(B=@n17) {constA=0.9, constB=0.9}, B=Sine(None=@n13) {period=1.25}) {constB=0.9}
  n21 = Multiply(A=Multiply(A=Add(A=@n16, B=@n18), B=@n19), B=Multiply(A=@n19, B=Ceil(None=@n20)))
  n22 = ConstantBiasScale(None=Sine(None=@n12) {period=0.075})
  n23 = Multiply(A=Ceil(None=AppendVector(A=@n14, B=@n15)), B=Multiply(A=Saturate(None=@n21), B=LinearInterpolate(Alpha=@n22) {constA=0.7}))
  n24 = ScalarParameter() {defaultValue=2, group=Noise, parameterName=NoiseMult}
  n25 = Multiply(
   A=Multiply(A=Power(Base=@n9.XY Texture) {constExponent=4}, B=ScalarParameter() {defaultValue=2, group=Noise, parameterName=NoiseMaskMult}) {constB=20}
   B=Add(A=ScalarParameter() {group=Noise, parameterName=NoiseAdd}, B=Multiply(A=@n23, B=@n24))
  )
  n26 = WorldPosition()
  n27 = ComponentMask(None=MaterialFunctionCall() {materialFunction=&ObjectPivotPoint}.Object Pivot Location) {b=true}
  n28 = Add(A=@n27, B=ScalarParameter() {defaultValue=300, parameterName=PulseHeight}) {constB=1600}
  n29 = ScalarParameter() {defaultValue=1, parameterName=TurnOn}
  n30 = ScalarParameter() {defaultValue=33, parameterName=PulseRadius}
  n31 = Multiply(A=@n30) {constB=0.1}
  n32 = Multiply(
   A=SphereMask(
    A=ComponentMask(None=@n26.XYZ) {b=true}
    B=LinearInterpolate(A=@n27, B=@n28, Alpha=@n29) {constA=-0.25, constB=1.2}
    Radius=LinearInterpolate(A=@n30, B=@n31, Alpha=@n29) {constA=500, constB=100}
   ) {hardnessPercent=0}
   B=OneMinus(None=@n29)
  )
  n33 = Add(A=Desaturation(None=Saturate(None=@n25)), B=Multiply(A=@n32) {constB=2})
  n34 = QualitySwitch(
   Default=MaterialFunctionCall(In=@n33, Contrast=ScalarParameter() {defaultValue=0.2, group=Noise, parameterName=NoiseContrast}) {materialFunction=&CheapContrast}.Result
   Low=Constant()
  )
  n35 = LinearInterpolate(B=ScalarParameter() {defaultValue=0.6, group=Noise, parameterName=NoiseFresnelXPMult}, Alpha=@n34) {constA=1, constB=0.5}
  n36 = Multiply(A=ScalarParameter() {defaultValue=1, parameterName=FresnelExp}, B=@n35)
  n37 = ScalarParameter() {defaultValue=0.5, parameterName=FresnelMin}
  n38 = Constant3Vector() {constant=<rgba 0 0 1 0>}
  n39 = CameraPositionWS()
  n40 = ActorPositionWS()
  n41 = Normalize(VectorInput=Multiply(A=Constant3Vector() {constant=<rgba 1 1 0 0>}, B=Subtract(A=@n39, B=@n40)))
  n42 = Normalize(VectorInput=Constant3Vector() {constant=<rgba 0 1 0 0>})
  n43 = DotProduct(A=@n41, B=Normalize(VectorInput=Constant3Vector() {constant=<rgba 1 0 0 0>}))
  n44 = Multiply(
   A=Divide(A=MaterialFunctionCall!(Normalized V1=@n41, Normalized V2=@n42) {materialFunction=/Engine/Functions/Engine_MaterialFunctions02/Math/AngleBetweenVectors}.Angle In Degrees) {constB=360}
   B=Multiply(A=Sign(None=@n43)) {constB=-1}
  )
  n45 = StaticSwitchParameter(
   True=MaterialFunctionCall(Rotation Axis=@n38, Rotation Angle=@n44) {materialFunction=&FixRotateAboutAxisNormals}.New World Space Vertex Normal
   False=PixelNormalWS()
  ) {parameterName=FaceCamera}
  n46 = ScalarParameter() {defaultValue=1, parameterName=EdgeFresnelXP}
  n47 = LinearInterpolate(Alpha=Fresnel(ExponentIn=Multiply(A=@n35, B=@n46), Normal=@n45) {baseReflectFraction=0, exponent=1}) {constA=5, constB=-0.4}
  n48 = Power(Base=Multiply(A=Fresnel(ExponentIn=@n36, BaseReflectFractionIn=@n37, Normal=@n45) {baseReflectFraction=0, exponent=0.5}, B=Saturate(None=@n47)))
  n49 = Saturate(None=@n34)
  n50 = Multiply(A=ScalarParameter() {defaultValue=-0.4, group=Noise, parameterName=NoiseBrightness}, B=@n49) {constB=0.1}
  n51 = Ceil(None=@n49)
  n52 = WorldPosition()
  n53 = CameraPositionWS()
  n54 = ActorPositionWS()
  n55 = ScalarParameter() {defaultValue=20000, parameterName=ScaleDistance}
  n56 = Saturate(None=Divide(A=Distance(A=@n53, B=@n54), B=@n55) {constB=1000})
  n57 = Multiply(
   A=Multiply(A=ScalarParameter() {defaultValue=200, parameterName=ScanlineTile}, B=LinearInterpolate(Alpha=@n51) {constA=1, constB=1.25})
   B=Divide(A=ComponentMask(None=@n52.XYZ) {b=true}, B=LinearInterpolate(B=@n55, Alpha=@n56) {constA=50, constB=100}) {constB=20}
  ) {constA=1}
  n58 = DotProduct(A=VertexNormalWS(), B=Constant3Vector() {constant=<rgba 0 0 1 0>})
  n59 = LinearInterpolate(A=ConstantBiasScale(None=Sine(None=@n57)), Alpha=Power(Base=Abs(None=@n58)) {constExponent=5}) {constB=0.25}
  n60 = Add(A=ScalarParameter() {defaultValue=0.7, parameterName=ScanlineColorMinInfluence}, B=@n59)
  n61 = ComponentMask(None=Subtract(A=WorldPosition().XYZ, B=ObjectPositionWS())) {b=true}
  n62 = ObjectRadius()
  n63 = Saturate(None=ConstantBiasScale(None=Divide(A=@n61, B=@n62)))
  n64 = LinearInterpolate(
   A=Multiply(A=Add(A=@n48, B=@n50), B=Saturate(None=@n60))
   B=LinearInterpolate(A=@n63, B=OneMinus(None=@n63), Alpha=ScalarParameter() {parameterName=FlipGradient})
   Alpha=ScalarParameter() {defaultValue=0.4, parameterName=VerticalGradientInflience}
  )
  n65 = Multiply(
   A=LinearInterpolate(
    A=ScalarParameter() {parameterName=MaskMin}
    Alpha=ChannelMaskParameter(None=@n5.RGB) {maskChannel=Green, parameterName=MaskChannel, defaultValue=<rgba 0 1 0 0>}.R
   )
   B=Add(A=Multiply(A=@n6, B=@n64), B=@n32) {constA=1}
  )
  n66 = TextureCoordinate()
  n67 = TextureSampleParameter2D(UVs=@n66) {group=VendettaPuck, parameterName=SlateUI, samplerSource=SSM_Clamp_WorldGroupSettings, texture=&Black}
  n68 = VectorParameter() {group=Texture, parameterName=TextureAdd}
  n69 = Add(A=Multiply(A=@n66, B=ScalarParameter() {defaultValue=0.5, group=Texture, parameterName=TextureSize}), B=AppendVector(A=@n68.R, B=@n68.R))
  n70 = TextureSampleParameter2D(UVs=@n69) {
   group=Texture
   parameterName=ProfileImage
   texture=/Game/UI/Foundation/Textures/Icons/Heroes/Athena/Soldier/T-Soldier-HID-005-Athena-Commando-M-L
  }
  n71 = StaticSwitchParameter(True=@n67.RGB, False=@n70.RGB) {group=VendettaPuck, parameterName=UseSlateUI}
  n72 = Desaturation(None=@n71, Fraction=Constant() {r=1})
  n73 = ScalarParameter() {defaultValue=0.25, group=VendettaPuck, parameterName=Profile-OpacityAdd}
  n74 = StaticSwitchParameter(True=@n67.A, False=@n70.A) {group=VendettaPuck, parameterName=UseSlateUI}
  n75 = Multiply(A=Clamp(None=Add(A=@n72, B=@n73)), B=@n74)
  n76 = Constant() {r=0.5}
  n77 = ScalarParameter() {defaultValue=0.5, group=VendettaPuck, parameterName=Profile-SphereMask-Radius}
  n78 = ScalarParameter() {defaultValue=0.65, group=VendettaPuck, parameterName=Profile-SphereMask-Hardness}
  n79 = StaticSwitchParameter(
   True=Multiply(A=ComponentMask(None=@n75) {r=true}, B=SphereMask(A=@n66, B=@n76, Radius=@n77, Hardness=@n78))
   False=TextureSampleParameter2D(UVs=@n69) {group=Texture, parameterName=2DHologramTexture, texture=/Game/Environments/Galileo/Textures/T_Galileo_Logo_Flatbed}.G
  ) {group=VendettaPuck, parameterName=VendettaPuck}
  n80 = Power(Base=@n79, Exp=@n36)
  n81 = StaticSwitchParameter(True=@n80, False=Saturate(None=Add(A=@n4, B=@n65))) {parameterName=2DHologram}
  n82 = ScalarParameter() {defaultValue=1, group=VendettaPuck, parameterName=SlateColor_LerpCurvePower}
  n83 = CurveAtlasRowParameter(CurveTime=@n81) {atlas=&AthenaCurveAtlas01, curve=/Game/Athena/LinearColorCurves/Curve_Galileo_Hologram, defaultValue=171}
  n84 = StaticSwitchParameter(
   True=LinearInterpolate(
    A=VectorParameter() {group=VendettaPuck, parameterName=SlateColor-Dark, defaultValue=<rgba 0 0 0 1>}.RGB
    B=VectorParameter() {group=VendettaPuck, parameterName=SlateColor-Bright, defaultValue=<rgba 0.0298 0.0866 0.2 1>}.RGB
    Alpha=Power(Base=@n81, Exp=@n82)
   )
   False=@n83
  ) {group=VendettaPuck, parameterName=UseSlateUI}
  n85 = TextureSampleParameter2D() {group=Diffuse Texture Overlay, parameterName=Diffuse Texture Overlay, texture=&WhiteSquareTexture}
  n86 = LinearInterpolate(
   A=@n85.RGB
   B=MaterialFunctionCall(In=@n85.RGB, Contrast=ScalarParameter() {group=Diffuse Texture Overlay, parameterName=Diffuse Contribution Cheap Contrast}) {materialFunction=&CheapContrast}.Result
   Alpha=ScalarParameter() {group=Diffuse Texture Overlay, parameterName=Diffuse Overlay Use Cheap Contrast}
  )
  n87 = ScalarParameter() {defaultValue=1, group=Diffuse Texture Overlay, parameterName=Diffuse Contribution Desaturation}
  n88 = LinearInterpolate(
   A=@n84
   B=Multiply(A=Desaturation(None=@n86, Fraction=@n87), B=@n84)
   Alpha=ScalarParameter() {group=Diffuse Texture Overlay, parameterName=Lerp Diffuse Overlay}
  )
  n89 = Clamp(None=LinearInterpolate(A=@n88, B=Multiply(A=@n71, B=@n74), Alpha=ScalarParameter() {defaultValue=0.25, group=VendettaPuck, parameterName=Profile-AddColor}))
  n90 = Multiply(
   A=LinearInterpolate(B=ParticleColor().RGB, Alpha=ScalarParameter() {parameterName=UseParticleColor}) {constA=1, desc=Workaround for skel mesh bug}
   B=Multiply(
    A=StaticSwitchParameter(True=@n89, False=@n88) {group=VendettaPuck, parameterName=VendettaPuck}
    B=ScalarParameter() {defaultValue=2, parameterName=EmissiveMult}
   ) {constB=2}
  )
  MaterialFunctionCall(Emissive=@n90) {materialFunction=&MF_FX_EmissiveAdjustments}
  n91 = ParticleColor()
  n92 = ScalarParameter() {parameterName=UseParticleAlpha}
  n93 = StaticSwitchParameter(
   True=Add(A=@n80, B=Multiply(A=@n79, B=@n50))
   False=DepthFade(Opacity=@n65, FadeDistance=ScalarParameter() {defaultValue=5, parameterName=FadeDistance})
  ) {parameterName=2DHologram}
  n94 = ScalarParameter() {defaultValue=0.6, parameterName=MinScanlineOpacity}
  n95 = Multiply(A=@n34) {constB=1.5}
  n96 = LinearInterpolate(A=Clamp(None=Subtract(A=@n94, B=@n95)) {minDefault=-0.25}, B=Add(A=@n95), Alpha=Power(Base=@n59) {constExponent=1.5}) {constA=0.5}
  n97 = Multiply(A=Saturate(None=Multiply(A=@n93, B=@n96)), B=ScalarParameter() {defaultValue=1.25, parameterName=OpacityMult})
  n98 = PixelDepth()
  n99 = ScalarParameter() {defaultValue=-0.25, parameterName=BiasSorting_Far}
  n100 = SceneTexture() {sceneTextureId=PPI_CustomDepth}
  n101 = Constant()
  n102 = ShadingPathSwitch(
   Default=If(A=Add(A=@n98.R, B=@n99), B=ComponentMask(None=@n100.Color) {r=true}, "A > B"=@n101, "A == B"=@n101, "A < B"=Constant() {r=1})
   Mobile=Constant() {r=1}
  ) {desc=Migrated from FeatureLevelSwitch}
  n103 = Time()
  n104 = ScalarParameter() {defaultValue=5, parameterName=StrobeSpeed}
  n105 = Multiply(A=Sine(None=Multiply(A=@n103, B=@n104)), B=ScalarParameter() {defaultValue=2, parameterName=StrobeHardness})
  Multiply(
   A=Multiply(A=LinearInterpolate(B=@n91.A, Alpha=@n92) {constA=1, desc=Workaround for skel mesh bug}, B=Multiply(A=@n97, B=@n102))
   B=LinearInterpolate(B=Clamp(None=@n105), Alpha=ScalarParameter() {parameterName=Strobe}) {constA=1}
  ) {constA=1}
  n106 = Normalize(VectorInput=Subtract(A=WorldPosition().XYZ, B=ObjectPositionWS()))
  n107 = Constant3Vector() {constant=<rgba 1 1 0 0>}
  n108 = ScalarParameter() {defaultValue=10, group=Noise, parameterName=WPOStrength}
  n109 = Multiply(A=ObjectRadius(), B=ScalarParameter() {defaultValue=0.025, group=Noise, parameterName=WPOStrengthFromObjectRadius}) {constB=0.05}
  n110 = Multiply(A=@n33, B=Multiply(A=Multiply(A=@n106, B=@n107), B=Add(A=@n108, B=@n109))) {constB=0}
  n111 = Add(A=@n110, B=RotateAboutAxis(NormalizedRotationAxis=@n38, RotationAngle=@n44, PivotPoint=@n40, Position=WorldPosition().XYZ))
  n112 = Multiply(A=Sine(None=Time()) {period=4}, B=VectorParameter() {parameterName=BobDistance, defaultValue=<rgba 0 0 160 0>}.RGB)
  Reroute(None=StaticSwitchParameter(True=Add(A=@n111, B=@n112), False=@n110) {parameterName=FaceCamera})
  Reroute(None=@n90)
 ]
}
