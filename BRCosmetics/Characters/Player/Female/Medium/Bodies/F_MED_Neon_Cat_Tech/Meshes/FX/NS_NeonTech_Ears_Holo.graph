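# Niagara effect for holographic ears on the Neon Cat Tech player cosmetic.
# Glow and Light each burst one persistent mesh at a system offset, tinted by user colors; Light adds two light renderers and stores a Z-mirrored second position.
# Ghosting spawns brief, fading pink meshes from the skeletal mesh as it moves, using spawn-per-unit.
# All three emitters set particle VisibilityTag from User.bisHeadHidden, letting renderers respond to head hiding.
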NiagaraSystem /BRCosmetics/Characters/Player/Female/Medium/Bodies/F_MED_Neon_Cat_Tech/Meshes/FX/NS_NeonTech_Ears_Holo
dependencies {usedDataInterfaces [&NiagaraDataInterfaceCurve], usedDynamicInputs [], usedModules [&SystemState]}
emitters {
 Ghosting {
  emitterUpdateScript [
   &EmitterState(Life Cycle Mode=::NewEnumerator1, Loop Behavior=::NewEnumerator1, Loop Duration Mode=::NewEnumerator1, UseLoopDelay=false)
   &SpawnBurst_Instantaneous(Spawn Count=1, Use Spawn Probability=false) {enabled=false}
   &SpawnPerUnit(
    Max Movement Threshold=310
    Module.Use Max Movement Threshold=true
    Module.Use Movement Tolerance=true
    Movement Tolerance=1
    Spawn Spacing=5
    Use Spawn Probability=false
   )
   &SpawnPerUnit(
    Max Movement Threshold=10000
    Module.Use Max Movement Threshold=true
    Module.Use Movement Tolerance=true
    Movement Tolerance=7
    Spawn Spacing=5
    Use Spawn Probability=false
   ) {enabled=false, module=SpawnPerUnit001}
   &SpawnRate(SpawnRate=50, Use Spawn Probability=false) {enabled=false}
  ]
  particleSpawnScript [
   &InitializeParticle(
    Color=&MultiplyLinearColor(Linear Color=<rgba 1 1 1 1>, Scale Factor=<0.5 0.5 0.5 1>)
    Lifetime=0.03
    Mesh Scale Mode=::NewEnumerator3
    UsePositionOffset=false
    Write Material Random=true
   )
   &InitialMeshOrientation(Orientation Vector=<0 0 1>, Rotation=<0 0.5 0>, Use Orientation Vector=true, Use Rotation Vector=true) {module=InitialMeshOrientation001}
   &DynamicMaterialParameters(
    Index 0 Param 1=1.2
    Index 0 Param 2=0.15
    Index 0 Param 3=1
    Index 0 Param 4=1
    Module.Param0WriteEnabled=true
    Module.Param1WriteEnabled=true
    Module.Param2WriteEnabled=true
    Module.Param3WriteEnabled=true
   )
   &SkeletalMeshLocation(Bone Sampling Mode=::NewEnumerator4, MaskSpawnGroup=false, Skeletal Mesh=$User.SkeletalMesh)
   &SystemLocation(MaskSpawnGroup=false, Offset=<14 2 0>) {enabled=false}
  ]
  particleUpdateScript [
   &ParticleState()
   &PendulumSetup(
    Module.Enable Spring Driver=false
    Pendulum Length=14
    Pendulum Pivot Offset Coordinate Space:ENiagaraCoordinateSpace=::Simulation
    Rigid Pendulum=false
   ) {enabled=false}
   &SolveForcesAndVelocity(Clamp Velocity=false, Limit Acceleration=false)
   &Color(Color=&MakeLinearColorFromVectorAndFloat(
     "Vector (RGB)"=<1 0.4 1>
     "Float (Alpha)"=&FloatFromCurve(FloatCurve=DataInterface() {propertyValues {Curve {keys [
          {arriveTangent=-3.131, interpMode=RCIM_Cubic, leaveTangent=-3.131, tangentMode=RCTM_User, time=0, value=1}
          {interpMode=RCIM_Cubic, tangentMode=RCTM_User, time=1, value=0}
         ]}}})
    ))
   &PendulumConstraint() {enabled=false}
   &CalculateAccurateVelocity() {enabled=false}
   &UpdateMeshOrientation(Rotation Coordinate Space=::NewEnumerator3, Rotation Vector=<1 0 0>) {enabled=false}
   /SetVariables(Particles.VisibilityTag=&MakeCustomIntFromBool(Bool=$User.bisHeadHidden, True INT=1, False INT=0)) {isSetParameters=true, module=SetVariables_#1}
  ]
  rendererClasses [&NiagaraMeshRendererProperties]
  renderers [{rendererClass=&NiagaraMeshRendererProperties, rendererIndex=0}]
 }
 Glow {
  emitterUpdateScript [
   &EmitterState(Life Cycle Mode=::NewEnumerator1, Loop Behavior=::NewEnumerator1, Loop Duration Mode=::NewEnumerator1, UseLoopDelay=false)
   &SpawnBurst_Instantaneous(Spawn Count=1, Use Spawn Probability=false)
  ]
  particleSpawnScript [
   &InitializeParticle(
    Color=&MultiplyLinearColor(Linear Color=$User.HologramColor, Scale Factor=<0.5 0.5 0.5 1>)
    Mesh Scale Mode=::NewEnumerator3
    UsePositionOffset=false
    Write Material Random=true
   )
   &InitialMeshOrientation(Rotation=<0 0.25 0>, Use Orientation Vector=false, Use Rotation Vector=true) {module=InitialMeshOrientation001}
   &DynamicMaterialParameters(
    Index 0 Param 1=1.2
    Index 0 Param 2=0.15
    Index 0 Param 3=1
    Index 0 Param 4=1
    Module.Param0WriteEnabled=true
    Module.Param1WriteEnabled=true
    Module.Param2WriteEnabled=true
    Module.Param3WriteEnabled=true
   )
   &SystemLocation(MaskSpawnGroup=false, Offset=<14 2 0>)
  ]
  particleUpdateScript [
   &ParticleState(Kill Particles When Lifetime Has Elapsed=false, Let Infinitely Lived Particles Die When Emitter Deactivates=true)
   &Color(Color=&MultiplyLinearColor(Linear Color=$User.Emissive Color, Scale Factor=<10 10 10 1>))
   /SetVariables(Particles.VisibilityTag=&MakeCustomIntFromBool(Bool=$User.bisHeadHidden, True INT=1, False INT=0)) {isSetParameters=true, module=SetVariables_#1}
  ]
  rendererClasses [&NiagaraMeshRendererProperties]
  renderers [{rendererClass=&NiagaraMeshRendererProperties, rendererIndex=0}]
 }
 Light {
  emitterUpdateScript [
   &EmitterState(Life Cycle Mode=::NewEnumerator1, Loop Behavior=::NewEnumerator1, Loop Duration Mode=::NewEnumerator1, UseLoopDelay=false)
   &SpawnBurst_Instantaneous(Spawn Count=1, Use Spawn Probability=false)
  ]
  particleSpawnScript [
   &InitializeParticle(
    Color=&MultiplyLinearColor(Linear Color=$User.HologramColor, Scale Factor=<5 5 5 1>)
    Mesh Scale Mode=::NewEnumerator3
    Mesh Uniform Scale=0.2
    UsePositionOffset=false
    Write Material Random=true
   )
   &InitialMeshOrientation(Rotation=<-0.05 -0.05 -0.25>, Use Orientation Vector=false, Use Rotation Vector=true) {module=InitialMeshOrientation001}
   &DynamicMaterialParameters(
    Index 0 Param 1=1.2
    Index 0 Param 2=0.15
    Index 0 Param 3=1
    Index 0 Param 4=1
    Module.Param0WriteEnabled=true
    Module.Param1WriteEnabled=true
    Module.Param2WriteEnabled=true
    Module.Param3WriteEnabled=true
   )
   &SystemLocation(MaskSpawnGroup=false, Offset=<15 1 8>)
   /SetVariables(Particles.Position2:Vector3f=&Multiply_Vector(A=Linked() {linkedVariable {name=Particles.Position, type=NiagaraPosition}}, B=<1 1 -1>)) {isSetParameters=true}
  ]
  particleUpdateScript [
   &ParticleState(Kill Particles When Lifetime Has Elapsed=false, Let Infinitely Lived Particles Die When Emitter Deactivates=true)
   &Color(Color=&MultiplyLinearColor(Linear Color=$User.Emissive Color, Scale Factor=<10 10 10 1>))
   /SetVariables(Particles.VisibilityTag=&MakeCustomIntFromBool(Bool=$User.bisHeadHidden, True INT=1, False INT=0)) {isSetParameters=true, module=SetVariables_#1}
  ]
  rendererClasses [&NiagaraMeshRendererProperties, &NiagaraLightRendererProperties]
  renderers [
   {rendererClass=&NiagaraMeshRendererProperties, rendererIndex=0}
   {rendererClass=&NiagaraLightRendererProperties, rendererIndex=1}
   {rendererClass=&NiagaraLightRendererProperties, rendererIndex=2}
  ]
 }
}
summary {
 systemName=NS_NeonTech_Ears_Holo
 emitters [
  {emitterName=Glow, rendererClasses [&NiagaraMeshRendererProperties]}
  {emitterName=Light, rendererClasses [&NiagaraMeshRendererProperties, &NiagaraLightRendererProperties]}
  {emitterName=Ghosting, rendererClasses [&NiagaraMeshRendererProperties]}
 ]
 userVariables [
  User.bisHeadHidden {type=NiagaraBool}
  User.CC_SecondaryColor {type=LinearColor, description="", defaultValue {struct=LinearColor, value=<rgba 0.7967 0.2723 1 1>}}
  User.Emissive Color {type=LinearColor, description="", defaultValue {struct=LinearColor, value=<rgba 0.0844 0.4659 1 1>}}
  User.HologramColor {type=LinearColor, description="", defaultValue {struct=LinearColor, value=<rgba 0.8632 0.0178 0 1>}}
  User.IsDiving {type=NiagaraFloat, defaultValue=0, description=""}
  User.MID_CarabusHologram {
   type=MaterialInterface
   description=""
   defaultValue {struct=NiagaraExt_VariableValue_Object, value {object=null, objectClass=/Script/Engine.MaterialInterface}}
  }
  User.MID_ElasticGlow {
   type=MaterialInterface
   description=""
   defaultValue {struct=NiagaraExt_VariableValue_Object, value {object=null, objectClass=&MaterialInterface}}
  }
  User.SkeletalMesh {
   type=NiagaraDataInterfaceSkeletalMesh
   defaultValue {
    struct=NiagaraExt_VariableValue_DataInterface
    value {dataInterface=&self:NiagaraDataInterfaceSkeletalMesh_0}
   }
  }
 ]
}
