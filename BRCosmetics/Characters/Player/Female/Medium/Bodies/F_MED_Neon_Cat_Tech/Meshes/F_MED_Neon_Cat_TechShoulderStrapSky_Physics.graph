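# Physics asset for the left shoulder strap on the Neon Cat Tech female medium body cosmetic, defining its dynamic-bone physics.
# Small-capsule master bones anchor both ends of a capsule chain with limited swing and tight twist, forming a closed loop.
# The masters link to each other and through the rear one to the clavicle with free rotation.
# Larger clavicle, neck and upper-arm capsules approximate the body around the strap.
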PhysicsAsset /BRCosmetics/Characters/Player/Female/Medium/Bodies/F_MED_Neon_Cat_Tech/Meshes/F_MED_Neon_Cat_TechShoulderStrapSky_Physics
bodies {
 Dyn_Master_ShoulderStrap_l [{shapeType=Capsule, length=0.505, radius=0.505, shapeName=Dyn_Master_ShoulderStrap_l_capsule}]
 Dyn_Master_Strap_Front_l [{shapeType=Capsule, length=0.505, radius=0.505, shapeName=Dyn_Master_Strap_Front_l_capsule}]
 Dyn_ShoulderStrap [
  {
   shapeType=Capsule
   length=6
   radius=1.2
   shapeName=Dyn_ShoulderStrap_capsule
   center=<2.2974 0.6964 -0.1866>
   rotation=<rot 90 18.166 15.5393>
  }
  {shapeType=Capsule, length=4, radius=1, shapeName=Dyn_ShoulderStrap_capsule, center=<7.0581 3.092 0.0154>, rotation=<rot 90 0 -58.047>}
 ]
 Dyn_ShoulderStrap_01 [{
   shapeType=Capsule
   length=6
   radius=1
   shapeName=Dyn_ShoulderStrap_01_capsule
   center=<4.9837 2.53 -0.1866>
   rotation=<rot 90 14.0363 -37.2781>
  }]
 Dyn_ShoulderStrap_02 [{
   shapeType=Capsule
   length=6
   radius=1.5
   shapeName=Dyn_ShoulderStrap_02_capsule
   center=<3.2489 0.3105 -0.1986>
   rotation=<rot 84.9962 10.0035 0>
  }]
 Dyn_ShoulderStrap_03 [{shapeType=Capsule, length=6, radius=1.5, shapeName=Dyn_ShoulderStrap_03_capsule, center=<1 0 0>, rotation=<rot 84.9997 -5.0003 0>}]
 Dyn_ShoulderStrap_l01 [{
   shapeType=Capsule
   length=8
   radius=1
   shapeName=Dyn_ShoulderStrap_l01_capsule
   center=<4.3004 1.9264 1.4078>
   rotation=<rot 79.9986 -159.9977 179.9999>
  }]
 Dyn_ShoulderStrap_l02 [{shapeType=Capsule, length=0.5, radius=0.5, shapeName=Dyn_ShoulderStrap_l02_capsule}]
 clavicle_l [{shapeType=Capsule, length=10, radius=4.5, shapeName=clavicle_l_capsule, center=<4.4612 3.0848 -0.0001>, rotation=<rot 90 0 17.0247>}]
 neck_02 [{shapeType=Capsule, length=3, radius=13, shapeName=neck_02_capsule, center=<4.7029 -3.978 -0.544>}]
 upperarm_twist_01_l [{shapeType=Capsule, length=20, radius=5.5, shapeName=upperarm_twist_01_l_capsule, center=<4.3933 -0.0582 -1.5985>, rotation=<rot 86.1394 0 0>}]
}
constraints [
 {bone1Name=Dyn_ShoulderStrap_l02, bone2Name=Dyn_Master_ShoulderStrap_l, swing1LimitDegrees=10, swing2LimitDegrees=50, twistLimitDegrees=1}
 {bone1Name=Dyn_ShoulderStrap_l01, bone2Name=Dyn_Master_Strap_Front_l, swing1LimitDegrees=30, swing2LimitDegrees=30, twistLimitDegrees=1}
 {bone1Name=Dyn_ShoulderStrap, bone2Name=Dyn_ShoulderStrap_l01, swing1LimitDegrees=30, swing2LimitDegrees=40, twistLimitDegrees=1}
 {bone1Name=Dyn_ShoulderStrap_01, bone2Name=Dyn_ShoulderStrap, swing1LimitDegrees=20, swing2LimitDegrees=40, twistLimitDegrees=1}
 {bone1Name=Dyn_ShoulderStrap_02, bone2Name=Dyn_ShoulderStrap_01, swing1LimitDegrees=20, swing2LimitDegrees=5, twistLimitDegrees=5}
 {bone1Name=Dyn_ShoulderStrap_03, bone2Name=Dyn_ShoulderStrap_02, swing1LimitDegrees=20, swing2LimitDegrees=35, twistLimitDegrees=1}
 {bone1Name=Dyn_ShoulderStrap_03, bone2Name=Dyn_ShoulderStrap_l02, swing1LimitDegrees=30, swing2LimitDegrees=30, twistLimitDegrees=1}
 {
  bone1Name=upperarm_twist_01_l
  bone2Name=clavicle_l
  swing1LimitDegrees=45
  swing1Motion=Free
  swing2LimitDegrees=45
  swing2Motion=Free
  twistLimitDegrees=45
  twistMotion=Free
 }
 {
  bone1Name=Dyn_Master_ShoulderStrap_l
  bone2Name=clavicle_l
  swing1LimitDegrees=45
  swing1Motion=Free
  swing2LimitDegrees=45
  swing2Motion=Free
  twistLimitDegrees=45
  twistMotion=Free
 }
 {
  bone1Name=Dyn_Master_Strap_Front_l
  bone2Name=Dyn_Master_ShoulderStrap_l
  swing1LimitDegrees=45
  swing1Motion=Free
  swing2LimitDegrees=45
  swing2Motion=Free
  twistLimitDegrees=45
  twistMotion=Free
 }
]
