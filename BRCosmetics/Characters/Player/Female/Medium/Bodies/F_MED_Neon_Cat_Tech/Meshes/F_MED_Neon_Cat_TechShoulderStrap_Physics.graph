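# Physics asset for the left shoulder strap on the F_MED_Neon_Cat_Tech female player cosmetic body.
# Capsule bodies form a strap chain joined by tight-twist, limited-swing constraints, anchored at both ends to the Dyn_Master front and shoulder bones.
# Free-rotation constraints link the two masters to each other and the shoulder master to the left clavicle, closing the loop.
# Larger clavicle, neck and upper-arm capsules approximate the surrounding body.
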PhysicsAsset /BRCosmetics/Characters/Player/Female/Medium/Bodies/F_MED_Neon_Cat_Tech/Meshes/F_MED_Neon_Cat_TechShoulderStrap_Physics
bodies {
 Dyn_Master_ShoulderStrap_l [{shapeType=Capsule, length=0.505, radius=0.505, shapeName=Dyn_Master_ShoulderStrap_l_capsule}]
 Dyn_Master_Strap_Front_l [{shapeType=Capsule, length=0.505, radius=0.505, shapeName=Dyn_Master_Strap_Front_l_capsule}]
 Dyn_ShoulderStrap [
  {
   shapeType=Capsule
   length=5.2347
   radius=1.2984
   shapeName=Dyn_ShoulderStrap_capsule
   center=<2.2974 0.6964 0.122>
   rotation=<rot 88.1465 -177.369 173.9353>
  }
  {
   shapeType=Capsule
   length=3.6861
   radius=1.3139
   shapeName=Dyn_ShoulderStrap_capsule
   center=<7.0581 3.092 0.2379>
   rotation=<rot 90 14.0362 -44.0108>
  }
 ]
 Dyn_ShoulderStrap_01 [{
   shapeType=Capsule
   length=5.0613
   radius=1.3914
   shapeName=Dyn_ShoulderStrap_01_capsule
   center=<4.9922 2.4667 -0.2377>
   rotation=<rot 89.0356 -128.6904 178.696>
  }]
 Dyn_ShoulderStrap_02 [{
   shapeType=Capsule
   length=6.0189
   radius=1.4811
   shapeName=Dyn_ShoulderStrap_02_capsule
   center=<3.1653 0.5114 -0.6941>
   rotation=<rot 81.9001 9.8647 0.2278>
  }]
 Dyn_ShoulderStrap_03 [{
   shapeType=Capsule
   length=3.7253
   radius=1.5284
   shapeName=Dyn_ShoulderStrap_03_capsule
   center=<2.0428 -0.2482 -0.2792>
   rotation=<rot 89.7948 8.8297 16.9933>
  }]
 Dyn_ShoulderStrap_l01 [{
   shapeType=Capsule
   length=7.9211
   radius=1.0789
   shapeName=Dyn_ShoulderStrap_l01_capsule
   center=<4.4145 1.9512 0.7462>
   rotation=<rot 79.9973 -159.9794 -176.1564>
  }]
 Dyn_ShoulderStrap_l02 [{shapeType=Capsule, length=0.4677, radius=0.6847, shapeName=Dyn_ShoulderStrap_l02_capsule}]
 clavicle_l [{shapeType=Capsule, length=10, radius=4.5, shapeName=clavicle_l_capsule, center=<4.4612 3.0848 -0.0001>, rotation=<rot 90 0 17.0247>}]
 neck_02 [{shapeType=Capsule, length=2, radius=9.5, shapeName=neck_02_capsule, center=<0.162 -3.978 -0.2616>}]
 upperarm_twist_01_l [{
   shapeType=Capsule
   length=21.6186
   radius=4.75
   shapeName=upperarm_twist_01_l_capsule
   center=<4.3965 -0.0583 -1.1545>
   rotation=<rot 89.0179 -178.6156 -178.6362>
  }]
}
constraints [
 {bone1Name=Dyn_ShoulderStrap_l02, bone2Name=Dyn_Master_ShoulderStrap_l, swing1LimitDegrees=10, swing2LimitDegrees=50, twistLimitDegrees=1}
 {bone1Name=Dyn_ShoulderStrap_l01, bone2Name=Dyn_Master_Strap_Front_l, swing1LimitDegrees=30, swing2LimitDegrees=30, twistLimitDegrees=1}
 {bone1Name=Dyn_ShoulderStrap, bone2Name=Dyn_ShoulderStrap_l01, swing1LimitDegrees=30, swing2LimitDegrees=40, twistLimitDegrees=1}
 {bone1Name=Dyn_ShoulderStrap_01, bone2Name=Dyn_ShoulderStrap, swing1LimitDegrees=20, swing2LimitDegrees=40, twistLimitDegrees=1}
 {bone1Name=Dyn_ShoulderStrap_02, bone2Name=Dyn_ShoulderStrap_01, swing1LimitDegrees=20, swing2LimitDegrees=5, twistLimitDegrees=5}
 {bone1Name=Dyn_ShoulderStrap_03, bone2Name=Dyn_ShoulderStrap_02, swing1LimitDegrees=20, swing2LimitDegrees=35, twistLimitDegrees=0}
 {bone1Name=Dyn_ShoulderStrap_03, bone2Name=Dyn_ShoulderStrap_l02, swing1LimitDegrees=30, swing2LimitDegrees=30, twistLimitDegrees=0}
 {
  bone1Name=upperarm_twist_01_l
  bone2Name=clavicle_l
  swing1LimitDegrees=45
  swing1Motion=Free
  swing2LimitDegrees=45
  swing2Motion=Free
  twistLimitDegrees=45
  twistMotion=Free
 }
 {
  bone1Name=Dyn_Master_ShoulderStrap_l
  bone2Name=clavicle_l
  swing1LimitDegrees=45
  swing1Motion=Free
  swing2LimitDegrees=45
  swing2Motion=Free
  twistLimitDegrees=45
  twistMotion=Free
 }
 {
  bone1Name=Dyn_Master_Strap_Front_l
  bone2Name=Dyn_Master_ShoulderStrap_l
  swing1LimitDegrees=45
  swing1Motion=Free
  swing2LimitDegrees=45
  swing2Motion=Free
  twistLimitDegrees=45
  twistMotion=Free
 }
]
