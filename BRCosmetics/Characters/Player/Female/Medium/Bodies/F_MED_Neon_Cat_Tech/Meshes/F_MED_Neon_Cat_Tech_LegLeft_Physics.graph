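# Physics asset for the left-leg mesh of the F_MED_Neon_Cat_Tech female player cosmetic, covering its dynamic pouch bones.
# Three capsule bodies sit on DynPouch_l, DynPouch_01l and DynPouch_01l1, with the widest on DynPouch_l.
# Two constraints link them in a chain with almost no twist, allowing mostly single-axis swing at the first joint and tighter swing at the tip.
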PhysicsAsset /BRCosmetics/Characters/Player/Female/Medium/Bodies/F_MED_Neon_Cat_Tech/Meshes/F_MED_Neon_Cat_Tech_LegLeft_Physics
bodies {
 DynPouch_01l [{shapeType=Capsule, length=4, radius=2, shapeName=DynPouch_01l_capsule, center=<3.5 0 0>, rotation=<rot 90 0 0>}]
 DynPouch_01l1 [{shapeType=Capsule, length=4, radius=2, shapeName=DynPouch_01l1_capsule, center=<3.5 0 0>, rotation=<rot 90 0 0>}]
 DynPouch_l [{shapeType=Capsule, length=4, radius=3.5, shapeName=DynPouch_l_capsule, center=<0.5 0 0>, rotation=<rot 90 0 0>}]
}
constraints [
 {bone1Name=DynPouch_01l, bone2Name=DynPouch_l, swing1LimitDegrees=50, swing2LimitDegrees=1, twistLimitDegrees=1}
 {bone1Name=DynPouch_01l1, bone2Name=DynPouch_01l, swing1LimitDegrees=20, swing2LimitDegrees=10, twistLimitDegrees=1}
]
