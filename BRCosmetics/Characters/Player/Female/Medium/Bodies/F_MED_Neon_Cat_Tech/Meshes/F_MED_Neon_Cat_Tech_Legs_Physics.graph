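# Physics asset for the legs mesh of the Neon Cat Tech female medium player cosmetic, covering its dangling pouch bones.
# Three capsule bodies form a chain from Dyn_Pouch through Dyn_Pouch_01r to Dyn_Pouch_02r, linked by two constraints.
# The first joint swings widely on one axis and is nearly locked on the other, while the second swings more narrowly on both.
# Both joints allow almost no twist.
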PhysicsAsset /BRCosmetics/Characters/Player/Female/Medium/Bodies/F_MED_Neon_Cat_Tech/Meshes/F_MED_Neon_Cat_Tech_Legs_Physics
bodies {
 Dyn_Pouch [{shapeType=Capsule, length=4, radius=3.5, shapeName=Dyn_Pouch_capsule, center=<0.5 0.2421 0>, rotation=<rot 90 0 0>}]
 Dyn_Pouch_01r [{shapeType=Capsule, length=4, radius=2, shapeName=Dyn_Pouch_01r_capsule, center=<3.5 0 0>, rotation=<rot 90 0 0>}]
 Dyn_Pouch_02r [{shapeType=Capsule, length=4, radius=2, shapeName=Dyn_Pouch_02r_capsule, center=<3.5 0 0>, rotation=<rot 90 0 0>}]
}
constraints [
 {bone1Name=Dyn_Pouch_01r, bone2Name=Dyn_Pouch, swing1LimitDegrees=50, swing2LimitDegrees=1, twistLimitDegrees=1}
 {bone1Name=Dyn_Pouch_02r, bone2Name=Dyn_Pouch_01r, swing1LimitDegrees=20, swing2LimitDegrees=10, twistLimitDegrees=1}
]
