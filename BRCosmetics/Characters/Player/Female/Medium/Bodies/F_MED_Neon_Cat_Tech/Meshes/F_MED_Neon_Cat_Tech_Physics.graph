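# Physics asset for the F_MED_Neon_Cat_Tech female player body, covering its dangling gear: pouches, straps, belt and jacket flap.
# Capsule bodies sit on Dyn_ bones, plus one on the left thigh.
# Constraints chain waist pouches, their straps, the body strap and jacket flap from small master bodies with tight twist limits.
# Left shoulder straps and the thigh-to-left-pouch link swing and twist freely, while a few bodies, including the chest pouch, have no constraint.
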PhysicsAsset /BRCosmetics/Characters/Player/Female/Medium/Bodies/F_MED_Neon_Cat_Tech/Meshes/F_MED_Neon_Cat_Tech_Physics
bodies {
 Dyn_BodyStrap1 [{shapeType=Capsule, length=4, radius=1, shapeName=Dyn_BodyStrap1_capsule, center=<0.5 0 0>, rotation=<rot 0 0 90>}]
 Dyn_ChestPouch_2 [{shapeType=Capsule, length=6, radius=5, shapeName=Dyn_ChestPouch_2_capsule, center=<0.5411 1.1272 0.7094>, rotation=<rot -0.0002 0 -13.1241>}]
 Dyn_JacketFlap1 [{shapeType=Capsule, length=4, radius=1.5, shapeName=Dyn_JacketFlap1_capsule, center=<0.5 -0.28 -0.06>, rotation=<rot 0 0 -12.0915>}]
 Dyn_JacketFlap2 [{shapeType=Capsule, length=4, radius=2, shapeName=Dyn_JacketFlap2_capsule, center=<0.5001 -1.4469 0>}]
 Dyn_JacketFlap3 [{shapeType=Capsule, length=4, radius=3, shapeName=Dyn_JacketFlap3_capsule, center=<-0.3202 -2.0324 0.0004>}]
 Dyn_MasterBodyStrap [{shapeType=Capsule, length=0.5, radius=0.5, shapeName=Dyn_MasterBodyStrap_capsule}]
 Dyn_MasterPouch_Waist_01 [{shapeType=Capsule, length=0.5, radius=0.5, shapeName=Dyn_MasterPouch_Waist_01_capsule}]
 Dyn_MasterPouch_Waist_02 [{shapeType=Capsule, length=0.5, radius=0.5, shapeName=Dyn_MasterPouch_Waist_02_capsule}]
 Dyn_MasterPouch_Waist_03 [{shapeType=Capsule, length=0.5, radius=0.5, shapeName=Dyn_MasterPouch_Waist_03_capsule}]
 Dyn_MasterPouch_Waist_04 [{shapeType=Capsule, length=0.5, radius=0.5, shapeName=Dyn_MasterPouch_Waist_04_capsule}]
 Dyn_MasterPouch_l [{shapeType=Capsule, length=0.5, radius=0.5, shapeName=Dyn_MasterPouch_l_capsule, rotation=<rot 0 -90 0>}]
 Dyn_Master_JacketFlap [{shapeType=Capsule, length=0.5, radius=0.5, shapeName=Dyn_Master_JacketFlap_capsule}]
 Dyn_Master_Pouch_l [{shapeType=Capsule, length=0.5, radius=0.5, shapeName=Dyn_Master_Pouch_l_capsule}]
 Dyn_Master_Pouch_r [{shapeType=Capsule, length=0.5, radius=0.5, shapeName=Dyn_Master_Pouch_r_capsule}]
 Dyn_Master_ShoulderStrap_l [{shapeType=Capsule, length=0.505, radius=0.505, shapeName=Dyn_Master_ShoulderStrap_l_capsule}]
 Dyn_Pouch_01 [{shapeType=Capsule, length=4, radius=3, shapeName=Dyn_Pouch_01_capsule, center=<0.5 0 0>, rotation=<rot 90 0 0>}]
 Dyn_Pouch_02 [{shapeType=Capsule, length=4, radius=3, shapeName=Dyn_Pouch_02_capsule, center=<0.5 -0.3804 -0.0001>, rotation=<rot -90 0 0>}]
 Dyn_Pouch_03 [{shapeType=Capsule, length=3, radius=2, shapeName=Dyn_Pouch_03_capsule, center=<0.5 -0.8834 0>, rotation=<rot 90 0 0>}]
 Dyn_Pouch_04 [{shapeType=Capsule, length=6, radius=4, shapeName=Dyn_Pouch_04_capsule}]
 Dyn_Pouch_Waist_Belt1 [{shapeType=Capsule, length=4, radius=1, shapeName=Dyn_Pouch_Waist_Belt1_capsule, center=<0.5 -0.8828 -0.0002>, rotation=<rot 90 0 0>}]
 Dyn_Pouch_Waist_Belt2 [{shapeType=Capsule, length=4, radius=1.3, shapeName=Dyn_Pouch_Waist_Belt2_capsule, center=<0.5 0 0>, rotation=<rot 90 0 0>}]
 Dyn_Pouch_Waist_Belt3 [{shapeType=Capsule, length=4, radius=1.3, shapeName=Dyn_Pouch_Waist_Belt3_capsule, center=<0.5 0 0>, rotation=<rot 90 0 0>}]
 Dyn_Pouch_Waist_BuckleStrap1 [{shapeType=Capsule, length=2, radius=1, shapeName=Dyn_Pouch_Waist_BuckleStrap1_capsule, center=<0.5 0 0>, rotation=<rot 90 0 0>}]
 Dyn_Pouch_Waist_BuckleStrap2 [{shapeType=Capsule, length=2, radius=1, shapeName=Dyn_Pouch_Waist_BuckleStrap2_capsule, center=<0.5 0 0>, rotation=<rot 90 0 0>}]
 Dyn_Pouch_Waist_BuckleStrap3 [{shapeType=Capsule, length=2, radius=1, shapeName=Dyn_Pouch_Waist_BuckleStrap3_capsule, center=<0.5 0 0>, rotation=<rot 90 0 0>}]
 Dyn_Pouch_Waist_Strap1 [{shapeType=Capsule, length=2, radius=0.5, shapeName=Dyn_Pouch_Waist_Strap1_capsule, center=<0.5 0 0>, rotation=<rot 90 0 0>}]
 Dyn_Pouch_Waist_Strap2 [{shapeType=Capsule, length=2, radius=0.5, shapeName=Dyn_Pouch_Waist_Strap2_capsule, center=<0.5 0 0>, rotation=<rot 90 0 0>}]
 Dyn_Pouch_Waist_Strap3 [{shapeType=Capsule, length=2, radius=0.5, shapeName=Dyn_Pouch_Waist_Strap3_capsule, center=<0.2591 0 0.0001>, rotation=<rot 90 0 0>}]
 Dyn_ShoulderStrap_l01 [{shapeType=Capsule, length=0.5, radius=0.5, shapeName=Dyn_ShoulderStrap_l01_capsule}]
 Dyn_ShoulderStrap_l02 [{shapeType=Capsule, length=0.5, radius=0.5, shapeName=Dyn_ShoulderStrap_l02_capsule}]
 thigh_l [{shapeType=Capsule, length=30, radius=7, shapeName=thigh_l_capsule, center=<-15.345 0.7861 1.1355>, rotation=<rot 0 -90.9491 -92.8225>}]
}
constraints [
 {bone1Name=Dyn_Pouch_02, bone2Name=Dyn_MasterPouch_Waist_02, swing1LimitDegrees=3, swing2LimitDegrees=1, twistLimitDegrees=1}
 {bone1Name=Dyn_Pouch_Waist_Strap1, bone2Name=Dyn_Pouch_02, swing1LimitDegrees=40, swing2LimitDegrees=1, twistLimitDegrees=1}
 {bone1Name=Dyn_Pouch_Waist_Strap2, bone2Name=Dyn_Pouch_Waist_Strap1, swing1LimitDegrees=40, swing2LimitDegrees=1, twistLimitDegrees=1}
 {bone1Name=Dyn_Pouch_Waist_Strap3, bone2Name=Dyn_Pouch_Waist_Strap2, swing1LimitDegrees=40, swing2LimitDegrees=1, twistLimitDegrees=1}
 {bone1Name=Dyn_BodyStrap1, bone2Name=Dyn_MasterBodyStrap, swing1LimitDegrees=2, swing2LimitDegrees=20, twistLimitDegrees=1}
 {bone1Name=Dyn_Pouch_Waist_BuckleStrap1, bone2Name=Dyn_Pouch_03, swing1LimitDegrees=40, swing2LimitDegrees=10, twistLimitDegrees=1}
 {
  bone1Name=Dyn_Pouch_Waist_BuckleStrap2
  bone2Name=Dyn_Pouch_Waist_BuckleStrap1
  swing1LimitDegrees=40
  swing2LimitDegrees=10
  twistLimitDegrees=1
 }
 {
  bone1Name=Dyn_Pouch_Waist_BuckleStrap3
  bone2Name=Dyn_Pouch_Waist_BuckleStrap2
  swing1LimitDegrees=40
  swing2LimitDegrees=10
  twistLimitDegrees=1
 }
 {bone1Name=Dyn_Pouch_03, bone2Name=Dyn_MasterPouch_Waist_03, swing1LimitDegrees=3, swing2LimitDegrees=1, twistLimitDegrees=1}
 {bone1Name=Dyn_Pouch_04, bone2Name=Dyn_MasterPouch_Waist_04, swing1LimitDegrees=3, swing2LimitDegrees=1, twistLimitDegrees=1}
 {bone1Name=Dyn_Pouch_01, bone2Name=Dyn_MasterPouch_Waist_01, swing1LimitDegrees=3, swing2LimitDegrees=1, twistLimitDegrees=1}
 {bone1Name=Dyn_Pouch_Waist_Belt1, bone2Name=Dyn_Pouch_01, swing1LimitDegrees=40, swing2LimitDegrees=40, twistLimitDegrees=1}
 {bone1Name=Dyn_Pouch_Waist_Belt2, bone2Name=Dyn_Pouch_Waist_Belt1, swing1LimitDegrees=40, swing2LimitDegrees=30, twistLimitDegrees=1}
 {bone1Name=Dyn_Pouch_Waist_Belt3, bone2Name=Dyn_Pouch_Waist_Belt2, swing1LimitDegrees=40, swing2LimitDegrees=10, twistLimitDegrees=1}
 {bone1Name=Dyn_JacketFlap1, bone2Name=Dyn_Master_JacketFlap, swing1LimitDegrees=20, swing2LimitDegrees=1, twistLimitDegrees=1}
 {bone1Name=Dyn_JacketFlap2, bone2Name=Dyn_JacketFlap1, swing1LimitDegrees=20, swing2LimitDegrees=1, twistLimitDegrees=1}
 {bone1Name=Dyn_JacketFlap3, bone2Name=Dyn_JacketFlap2, swing1LimitDegrees=20, swing2LimitDegrees=1, twistLimitDegrees=1}
 {
  bone1Name=Dyn_Master_Pouch_l
  bone2Name=thigh_l
  swing1LimitDegrees=45
  swing1Motion=Free
  swing2LimitDegrees=45
  swing2Motion=Free
  twistLimitDegrees=45
  twistMotion=Free
 }
 {
  bone1Name=Dyn_ShoulderStrap_l01
  bone2Name=Dyn_Master_ShoulderStrap_l
  swing1LimitDegrees=45
  swing1Motion=Free
  swing2LimitDegrees=45
  swing2Motion=Free
  twistLimitDegrees=45
  twistMotion=Free
 }
 {
  bone1Name=Dyn_ShoulderStrap_l02
  bone2Name=Dyn_Master_ShoulderStrap_l
  swing1LimitDegrees=45
  swing1Motion=Free
  swing2LimitDegrees=45
  swing2Motion=Free
  twistLimitDegrees=45
  twistMotion=Free
 }
]
