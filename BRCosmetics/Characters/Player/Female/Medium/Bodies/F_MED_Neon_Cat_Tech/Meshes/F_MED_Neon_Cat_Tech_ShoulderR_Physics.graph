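# Physics asset for the ShoulderR mesh of the Neon Cat Tech medium female player cosmetic, covering a chain of left-suffixed dynamic pouch bones.
# Five capsule bodies run from a tiny master body through a larger pouch body to three small segments, each constrained to its predecessor.
# The pouch is held nearly rigid to the master, while the segments below swing within moderate limits with almost no twist.
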PhysicsAsset /BRCosmetics/Characters/Player/Female/Medium/Bodies/F_MED_Neon_Cat_Tech/Meshes/F_MED_Neon_Cat_Tech_ShoulderR_Physics
bodies {
 Dyn_MasterPouch_l [{shapeType=Capsule, length=0.5, radius=0.5, shapeName=Dyn_MasterPouch_l_capsule, rotation=<rot 0 -90 0>}]
 Dyn_Pouch_l [{shapeType=Capsule, length=5, radius=4, shapeName=Dyn_Pouch_l_capsule, center=<0.9924 1.0735 0>, rotation=<rot 90 0 0>}]
 Dyn_Pouch_l1 [{shapeType=Capsule, length=2.5, radius=1, shapeName=Dyn_Pouch_l1_capsule, center=<0.5 -0.3825 0.0001>, rotation=<rot 90 0 0>}]
 Dyn_Pouch_l2 [{shapeType=Capsule, length=2.5, radius=1, shapeName=Dyn_Pouch_l2_capsule, center=<0.5 -0.1893 0>, rotation=<rot 90 0 0>}]
 Dyn_Pouch_l3 [{shapeType=Capsule, length=2.5, radius=1, shapeName=Dyn_Pouch_l3_capsule, center=<0.5 0 0>, rotation=<rot 90 0 0>}]
}
constraints [
 {bone1Name=Dyn_Pouch_l, bone2Name=Dyn_MasterPouch_l, swing1LimitDegrees=3, swing2LimitDegrees=1, twistLimitDegrees=1}
 {bone1Name=Dyn_Pouch_l1, bone2Name=Dyn_Pouch_l, swing1LimitDegrees=30, swing2LimitDegrees=20, twistLimitDegrees=1}
 {bone1Name=Dyn_Pouch_l2, bone2Name=Dyn_Pouch_l1, swing1LimitDegrees=20, swing2LimitDegrees=20, twistLimitDegrees=1}
 {bone1Name=Dyn_Pouch_l3, bone2Name=Dyn_Pouch_l2, swing1LimitDegrees=20, swing2LimitDegrees=20, twistLimitDegrees=1}
]
