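# Animation Blueprint for the Neon Cat Tech female body cosmetic, posing and simulating its straps and pouches atop the parent mesh's copied pose.
# A skydiving blend bypasses the deform Control Rigs and shoulder-strap look-at used otherwise.
# Arm-raise alphas, computed each update from upperarm-to-clavicle deltas, drive additive bone offsets, alongside a crouch pouch nudge.
# Chained RigidBody nodes then simulate body regions in bone space, including a skydive-only strap pass.
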AnimBlueprint /BRCosmetics/Characters/Player/Female/Medium/Bodies/F_MED_Neon_Cat_Tech/Meshes/F_MED_Neon_Cat_Tech_AnimBP
target_skeleton=&/F_MED_Neon_Cat_Tech_Skeleton
graphs [
 AnimGraph {nodes [
   "Constraint - Dyn_ShoulderStrap_02 (R)"(Alpha=1, ConstraintWeights_0=1) {
    alphaBoolBlend {blendOption=Linear}
    boneToModify {boneName=Dyn_ShoulderStrap_02}
    constraintSetup [{transformType=Rotation, targetBone {boneName=upperarm_twist_02_l}}]
    constraintWeights [1]
   }
   "Transform (Modify) Bone - Bone: Dyn_Master_ShoulderStrap_l"!(ComponentPose, bAlphaBoolEnabled:bool=True, Translation="0,0,0", Rotation="10,10,0", Scale="1,1,1") -> (Pose) {
    alphaInputType=Bool
    rotationMode=BMM_Additive
    rotationSpace=BCS_BoneSpace
    translationMode=BMM_Additive
    translationSpace=BCS_BoneSpace
    alphaBoolBlend {blendInTime=0.1, blendOption=Linear, blendOutTime=0.1}
    alphaScaleBiasClamp {bInterpResult=true, interpSpeedDecreasing=1, interpSpeedIncreasing=50}
    boneToModify {boneName=Dyn_Master_ShoulderStrap_l}
    rotation {pitch=10, yaw=10}
   }
   n1 = Get Right Arm Up() -> (Right Arm Up:double)
   n2 = "Transform (Modify) Bone - Bone: Dyn_MasterPouch_l"(
    ComponentPose=Local To Component(LocalPose=Copy Pose From Mesh() {bUseAttachedParent=true})
    Alpha=@n1
    Translation="0,0,1"
    Rotation="0,0,0"
    Scale="1,1,1"
   ) {
    rotationMode=BMM_Additive
    rotationSpace=BCS_BoneSpace
    translationMode=BMM_Additive
    translationSpace=BCS_BoneSpace
    alphaBoolBlend {blendInTime=0.1, blendOption=Linear, blendOutTime=20}
    boneToModify {boneName=Dyn_MasterPouch_l}
    translation {z=1}
   }
   n3 = Component To Local(ComponentPose=@n2)
   n4 = Control Rig(Source=Copy Pose From Mesh() {bUseAttachedParent=true}) {
    alphaBoolBlend {blendOption=Linear}
    controlRigAssetReference {blueprintRigClass=&/F_MED_Neon_Cat_Tech_Deform_CtrlRig_sculpters.F_MED_Neon_Cat_Tech_Deform_CtrlRig_sculpters_C}
   }
   n5 = Control Rig(Source=@n4) {
    alphaBoolBlend {blendOption=Linear}
    controlRigAssetReference {blueprintRigClass=&/F_MED_Neon_Cat_Tech_Deform_CtrlRig_twists.F_MED_Neon_Cat_Tech_Deform_CtrlRig_twists_C}
   }
   n6 = "Transform (Modify) Bone - Bone: Dyn_Master_ShoulderStrap_l"!(
    ComponentPose=Local To Component(LocalPose=@n5)
    Alpha:float=Get Left Arm Up() -> (Left Arm Up:double)
    Translation="2.549,-3.931,0.000"
    Rotation="0,0,0"
    Scale="1,1,1"
   ) -> (Pose) {
    rotationMode=BMM_Additive
    rotationSpace=BCS_BoneSpace
    translationMode=BMM_Additive
    translationSpace=BCS_BoneSpace
    alphaBoolBlend {blendOption=Linear}
    alphaScaleBiasClamp {interpSpeedDecreasing=1, interpSpeedIncreasing=50}
    boneToModify {boneName=Dyn_Master_ShoulderStrap_l}
    translation=<2.549 -3.931>
   }
   n7 = Look At - Bone: Dyn_Master_ShoulderStrap_l(ComponentPose=@n6, Alpha=1) {
    alphaBoolBlend {blendOption=Linear}
    alphaScaleBiasClamp {interpSpeedDecreasing=1, interpSpeedIncreasing=50}
    boneToModify {boneName=Dyn_Master_ShoulderStrap_l}
    lookAtLocation=<13.2369 -2.1563 11.4819>
    lookAtTarget {boneReference {boneName=upperarm_twist_01_l}}
    lookAt_Axis {axis=<1 0>}
    lookUp_Axis {axis=<0 -1>}
   }
   n8 = "Transform (Modify) Bone - Bone: Dyn_MasterPouch_l"(ComponentPose=@n7, Alpha=@n1, Translation="0,0,1", Rotation="0,0,80", Scale="1,1,1") {
    rotationMode=BMM_Additive
    rotationSpace=BCS_BoneSpace
    translationMode=BMM_Additive
    translationSpace=BCS_BoneSpace
    alphaBoolBlend {blendOption=Linear}
    boneToModify {boneName=Dyn_MasterPouch_l}
    rotation {roll=80}
    translation {z=1}
   }
   n9 = Component To Local(ComponentPose=@n8)
   n10 = Get bIsSkyDiving()
   n11 = "Transform (Modify) Bone - Bone: Dyn_Master_Strap_Front_l"(
    ComponentPose=Local To Component(LocalPose=Blend Poses by bool(BlendPose_0=@n3, BlendPose_1=@n9, BlendTime_0=0.2, BlendTime_1=0.2, bActiveValue=@n10) {blendType=Linear})
    Alpha=Get Left Arm Up()
    Translation="0,-0.5,0"
    Rotation="0,0,0"
    Scale="1,1,1"
   ) {
    rotationMode=BMM_Additive
    rotationSpace=BCS_BoneSpace
    translationMode=BMM_Additive
    translationSpace=BCS_BoneSpace
    alphaBoolBlend {blendOption=Linear}
    alphaScaleBiasClamp {bInterpResult=true, interpSpeedDecreasing=1, interpSpeedIncreasing=50}
    boneToModify {boneName=Dyn_Master_Strap_Front_l}
    translation {y=-0.5}
   }
   n12 = "Transform (Modify) Bone - Bone: Dyn_MasterPouch_Waist_01"!(ComponentPose=@n11, bAlphaBoolEnabled:bool=Get bIsCrouching(), Translation="0,0,2", Rotation="0,0,0", Scale="1,1,1") -> (Pose) {
    alphaInputType=Bool
    rotationMode=BMM_Additive
    rotationSpace=BCS_BoneSpace
    translationMode=BMM_Additive
    translationSpace=BCS_BoneSpace
    alphaBoolBlend {blendOption=Linear}
    alphaScaleBiasClamp {bInterpResult=true, interpSpeedDecreasing=1, interpSpeedIncreasing=50}
    boneToModify {boneName=Dyn_MasterPouch_Waist_01}
    translation {z=2}
   }
   n13 = RigidBody(ComponentPose=@n12) {
    bFreezeIncomingPoseOnStart=true
    bOverrideWorldGravity=true
    overridePhysicsAsset=&/F_MED_Neon_Cat_Tech_Physics
    simulationSpace=BaseBoneSpace
    alphaBoolBlend {blendOption=Linear}
    baseBoneRef {boneName=pelvis}
    componentAppliedLinearAccClamp=<2000 2000 2000>
    componentLinearAccScale=<0.3 0.3 0.3>
    componentLinearVelScale=<0.3 0.3 0.3>
    overrideWorldGravity {z=-980}
   }
   n14 = Get bIsSkyDiving()
   n15 = "Transform (Modify) Bone - Bone: Dyn_Master_ShoulderStrap_l"!(ComponentPose=@n13, bAlphaBoolEnabled:bool=@n14, Translation="0,0,0", Rotation="-20,-20,0", Scale="1,1,1") -> (Pose) {
    alphaInputType=Bool
    rotationMode=BMM_Additive
    rotationSpace=BCS_BoneSpace
    translationMode=BMM_Additive
    translationSpace=BCS_BoneSpace
    alphaBoolBlend {blendInTime=0.1, blendOption=Linear, blendOutTime=0.1}
    alphaScaleBiasClamp {bInterpResult=true, interpSpeedDecreasing=1, interpSpeedIncreasing=50}
    boneToModify {boneName=Dyn_Master_ShoulderStrap_l}
    rotation {pitch=-20, yaw=-20}
   }
   n16 = RigidBody!(ComponentPose=@n15, bAlphaBoolEnabled:bool=@n14, ExternalForce) -> (Pose) {
    alphaInputType=Bool
    bClampLinearTranslationLimitToRefPose=true
    bOverrideWorldGravity=true
    overridePhysicsAsset=&/F_MED_Neon_Cat_TechShoulderStrapSky_Physics
    simulationSpace=BaseBoneSpace
    alphaBoolBlend {blendInTime=0.1, blendOption=Linear, blendOutTime=0.1}
    baseBoneRef {boneName=upperarm_l}
    componentAppliedLinearAccClamp=<1000 1000 1000>
    componentLinearAccScale=<0.05 0.05 0.05>
    componentLinearVelScale=<0.05 0.05 0.05>
    overrideWorldGravity {z=-400}
   }
   n17 = RigidBody!(ComponentPose=@n16, Alpha:float, ExternalForce) -> (Pose) {
    bClampLinearTranslationLimitToRefPose=true
    bOverrideWorldGravity=true
    overridePhysicsAsset=&/F_MED_Neon_Cat_TechShoulderStrap_Physics
    simulationSpace=BaseBoneSpace
    alphaBoolBlend {blendOption=Linear}
    baseBoneRef {boneName=upperarm_l}
    componentAppliedLinearAccClamp=<1000 1000 1000>
    componentLinearAccScale=<0.1 0.1 0.1>
    componentLinearVelScale=<0.1 0.1 0.1>
    overrideWorldGravity {z=-980}
   }
   n18 = RigidBody(ComponentPose=@n17) {
    bClampLinearTranslationLimitToRefPose=true
    bOverrideWorldGravity=true
    overridePhysicsAsset=&/F_MED_Neon_Cat_Tech_ShoulderR_Physics
    simulationSpace=BaseBoneSpace
    alphaBoolBlend {blendOption=Linear}
    baseBoneRef {boneName=upperarm_twist_01_r}
    componentAppliedLinearAccClamp=<1000 1000 1000>
    componentLinearAccScale=<0.35 0.1 0.1>
    componentLinearVelScale=<0.35 0.1 0.1>
    overrideWorldGravity {z=-980}
   }
   n19 = RigidBody(ComponentPose=@n18) {
    bClampLinearTranslationLimitToRefPose=true
    bOverrideWorldGravity=true
    overridePhysicsAsset=&/F_MED_Neon_Cat_Tech_Legs_Physics
    simulationSpace=BaseBoneSpace
    alphaBoolBlend {blendOption=Linear}
    baseBoneRef {boneName=thigh_twist_01_r}
    componentAppliedLinearAccClamp=<2000 2000 2000>
    componentLinearAccScale=<0.35 0.35 0.35>
    componentLinearVelScale=<0.35 0.35 0.35>
    overrideWorldGravity {z=-980}
   }
   n20 = RigidBody(ComponentPose=@n19) {
    bClampLinearTranslationLimitToRefPose=true
    bOverrideWorldGravity=true
    overridePhysicsAsset=&/F_MED_Neon_Cat_Tech_LegLeft_Physics
    simulationSpace=BaseBoneSpace
    alphaBoolBlend {blendOption=Linear}
    baseBoneRef {boneName=thigh_twist_01_l}
    componentAppliedLinearAccClamp=<2000 2000 2000>
    componentLinearAccScale=<0.35 0.35 0.35>
    componentLinearVelScale=<0.35 0.35 0.35>
    overrideWorldGravity {z=-980}
   }
   Output Pose(Result=Component To Local(ComponentPose=@n20)) {layerGroup=None}
   Get Left Arm Rifle() -> (Left Arm Rifle:double)
   Get bIsSkydiveFloating()
  ]}
 EventGraph {nodes [
   Event BlueprintUpdateAnimation()
   > n1 = Sequence!(execute) -> (then_0, then_1, then_2:Exec)
   n2 = Get Skel Mesh()
   > Set Right Arm Up(Right Arm Up:double=GetAlphaFromReferenceTransformDelta(SkeletalMeshComponent=@n2, BoneName=upperarm_r, BaseName=clavicle_r, InputRange=(X=80,Y=116), TwistAxis="0,0,1")) -> (Output_Get:double)
   Set Left Arm Up(
    execute=@n1.then_1
    Left Arm Up:double=GetAlphaFromReferenceTransformDelta(SkeletalMeshComponent=@n2, BoneName=upperarm_l, BaseName=clavicle_l, InputRange=(X=50,Y=76), TwistAxis="0,0,1")
   ) -> (Output_Get:double)
   Event BlueprintInitializeAnimation()
   > Set Skel Mesh(Skel Mesh=GetOwnerSkeletalMeshComponent())
   Get bRecentlyFired() -> (bRecentlyFired:bool)
   Get bIsSkydiveFloating()
   Set Left Arm Rifle(Left Arm Rifle:double=GetAlphaFromReferenceTransformDelta(SkeletalMeshComponent=@n2, BoneName=upperarm_l, BaseName=clavicle_l, InputRange=(X=13,Y=21), TwistAxis="0,0,1")) -> (Output_Get:double)
  ]}
]
variables [
 Left Arm Rifle {type=number}
 Left Arm Up {type=number}
 Right Arm Up {type=number}
 Skel Mesh {type=SkeletalMeshComponent}
]
