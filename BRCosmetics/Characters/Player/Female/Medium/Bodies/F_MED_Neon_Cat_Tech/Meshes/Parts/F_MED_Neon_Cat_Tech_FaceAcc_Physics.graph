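# Physics asset for the face accessory mesh of the female medium Neon Cat Tech player cosmetic, covering hood and shoulder bones.
# Capsule bodies sit on both clavicles, the right upper arm, a tiny master hood bone and the right hood bone, which carries two capsules.
# One constraint holds the right hood to the master hood within very tight angular limits; another leaves the right upper arm rotating freely on its clavicle.
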PhysicsAsset /BRCosmetics/Characters/Player/Female/Medium/Bodies/F_MED_Neon_Cat_Tech/Meshes/Parts/F_MED_Neon_Cat_Tech_FaceAcc_Physics
bodies {
 Dyn_Hood_r [
  {shapeType=Capsule, length=10, radius=5, shapeName=Dyn_Hood_r_capsule, center=<-3.3917 1.9017 0.0011>}
  {shapeType=Capsule, length=10, radius=5, shapeName=Dyn_Hood_r_capsule, center=<-1.1112 3.965 6.4682>, rotation=<rot -14.5097 -20.7252 56.485>}
 ]
 Dyn_MasterHood [{shapeType=Capsule, length=0.5, radius=0.5, shapeName=Dyn_MasterHood_capsule}]
 clavicle_l [{shapeType=Capsule, length=10, radius=6.5, shapeName=clavicle_l_capsule, center=<3.2108 -1.3434 -1.1231>, rotation=<rot 90 -0.0396 -0.0396>}]
 clavicle_r [{shapeType=Capsule, length=10, radius=6.5, shapeName=clavicle_r_capsule, center=<-4.7939 0.9203 0.8218>, rotation=<rot 90 0 0>}]
 upperarm_r [{shapeType=Capsule, length=15, radius=4, shapeName=upperarm_r_capsule, center=<-9.8598 -0.3005 0.2788>, rotation=<rot 90 0 0>}]
}
constraints [
 {bone1Name=Dyn_Hood_r, bone2Name=Dyn_MasterHood, swing1LimitDegrees=1, swing2LimitDegrees=1, twistLimitDegrees=1}
 {
  bone1Name=upperarm_r
  bone2Name=clavicle_r
  swing1LimitDegrees=45
  swing1Motion=Free
  swing2LimitDegrees=45
  swing2Motion=Free
  twistLimitDegrees=45
  twistMotion=Free
 }
]
